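# Looping ambient sound cue for the MistRaven Spires story quest in the Season 36 battle pass, routed to the quests sound class.
# A single looping wave feeds a modulator for slight pitch variation, then a one-input mixer.
# It borrows the chest attenuation asset with falloff and shape overrides.
# Overridden concurrency allows one instance per owner and blocks new ones while it plays.
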SoundCue /BattlePassS36_Quests/Audio/SFX/Quest_S36_Story_MistRaven_Spires/Quest_S36_Story_MistRaven_Spires_Ambient_Loop_Cue
root=Mixer_0
nodes [
 {id=Mixer_0, children [Modulator_0], properties {inputVolume [1]}}
 {id=Modulator_0, children [WavePlayer_0], properties {pitchMin=0.97}}
 {
  id=WavePlayer_0
  properties {
   bLooping=true
   soundWaveAssetPtr=/BattlePassS36_Quests/Audio/SFX/Quest_S36_Story_MistRaven_Spires/Waves/Quest_S36_Story_MistRaven_Spire_Ambient_Loop
  }
 }
]
properties {
 attenuationSettings=/Game/Sounds/Attenuation/Fort_Chests_Att
 bOverrideConcurrency=true
 maxDistance=1300
 soundClassObject=&Fort_Quests
 volumeMultiplier=1.25
 attenuationOverrides {falloffDistance=650, attenuationShapeExtents {x=150}}
 concurrencyOverrides {bLimitToOwner=true, maxCount=1, resolutionRule=PreventNew}
}
